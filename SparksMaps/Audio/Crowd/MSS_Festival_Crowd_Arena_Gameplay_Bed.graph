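# Stereo one-shot MetaSound source for the festival arena crowd gameplay ambience bed in Sparks, assigned to the in-game SFX sound class.
# On play, a random get from the Wavs array, which defaults to one wave, feeds a stereo wave player.
# A decibel gain input scales both channels, and the player's finish trigger ends the source.
# Concurrency allows one instance per owner, and a song-volume-follow control bus modulates volume.
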MetaSoundSource /SparksMaps/Audio/Crowd/MSS_Festival_Crowd_Arena_Gameplay_Bed
inputs [
 GainDecibels {type=Float}
 Input {type=Float}
 UE.Source.OnPlay {type=Trigger}
 Wavs {type=WaveAsset:Array}
]
interfaces [UE.OutputFormat.Stereo:1.0, UE.Source.OneShot:1.0, UE.Source:1.1]
outputs [
 "UE.OutputFormat.Stereo.Audio:0" {type=Audio}
 "UE.OutputFormat.Stereo.Audio:1" {type=Audio}
 UE.Source.OneShot.OnFinished {type=Trigger}
]
pages [Default {
  input_defaults {GainDecibels=0, Input=0, UE.Source.OnPlay=null, Wavs [&/Festival_Crowd_Arena_Amb_Gameplay]}
  nodes [
   DecibelstoLinearGain_0 = UE.Decibels to Linear Gain:Float(Decibels=^GainDecibels)
   ArrayRandomGet_0 = Array.Random Get:WaveAsset:Array(Next=^UE.Source.OnPlay, In Array=^Wavs)
   WavePlayer_0 = UE.Wave Player:Stereo(Play=@ArrayRandomGet_0.On Next, Wave Asset=@ArrayRandomGet_0.Value) {also {defaults {Loop="<MetaSoundBuilderResult.SUCCEEDED: 0>"}, version [1,1]}}
   Multiply_0 = UE.Multiply:Audio by Float(AdditionalOperands=@DecibelstoLinearGain_0.Linear Gain, PrimaryOperand=@WavePlayer_0.Out Right)
   Multiply_1 = UE.Multiply:Audio by Float(AdditionalOperands=@DecibelstoLinearGain_0.Linear Gain, PrimaryOperand=@WavePlayer_0.Out Left)
  ]
  links [
   {dst="@UE.OutputFormat.Stereo.Audio:1", src=Multiply_0.Out}
   {dst="@UE.OutputFormat.Stereo.Audio:0", src=Multiply_1.Out}
   {dst="@UE.Source.OneShot.OnFinished", src=WavePlayer_0.On Finished}
  ]
 }]
properties {
 assetImportData=&self:AssetImportData
 bOverrideConcurrency=true
 numChannels=2
 outputFormat=Stereo
 sampleRate=48000
 soundClassObject=&Fort_SFX_InGame
 volume=0.15
 concurrencyOverrides {bLimitToOwner=true, maxCount=1}
 modulationSettings {
  volumeRouting=Union
  volumeModulationDestination {bEnableModulation=true, modulators [/SparksCore/Audio/Modulation/CB_Sparks_SongVolumeFollow]}
 }
}
