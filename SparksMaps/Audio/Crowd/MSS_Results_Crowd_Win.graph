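# Stereo MetaSound source for the winning results-screen crowd in SparksMaps, using an arena qualify crowd loop.
# On play, a random get picks from the Waves array, which holds one asset by default, and triggers a stereo wave player wired straight to the outputs.
# It plays when silent, with concurrency limited to one instance per owner, stopping the oldest.
# The float Input is unconnected.
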MetaSoundSource /SparksMaps/Audio/Crowd/MSS_Results_Crowd_Win
inputs [
 Input {type=Float}
 UE.Source.OnPlay {type=Trigger}
 Waves {type=WaveAsset:Array}
]
interfaces [UE.OutputFormat.Stereo:1.0, UE.Source:1.1]
outputs [
 "UE.OutputFormat.Stereo.Audio:0" {type=Audio}
 "UE.OutputFormat.Stereo.Audio:1" {type=Audio}
]
pages [Default {
  input_defaults {Input=0, UE.Source.OnPlay=null, Waves [&/Festival_Crowd_Arena_Results_Qualify_Loop]}
  nodes [
   ArrayRandomGet_0 = Array.Random Get:WaveAsset:Array(Next=^UE.Source.OnPlay, In Array=^Waves)
   WavePlayer_0 = UE.Wave Player:Stereo(Play=@ArrayRandomGet_0.On Next, Wave Asset=@ArrayRandomGet_0.Value) {also {defaults {Loop="<MetaSoundBuilderResult.SUCCEEDED: 0>"}, version [1,1]}}
  ]
  links [{dst="@UE.OutputFormat.Stereo.Audio:0", src=WavePlayer_0.Out Left}, {dst="@UE.OutputFormat.Stereo.Audio:1", src=WavePlayer_0.Out Right}]
 }]
properties {
 assetImportData=&self:AssetImportData
 bLooping=true
 bOverrideConcurrency=true
 duration=10000
 numChannels=2
 outputFormat=Stereo
 sampleRate=48000
 soundClassObject=&Fort_SFX_InGame
 virtualizationMode=PlayWhenSilent
 volume=0.4
 concurrencyOverrides {bLimitToOwner=true, maxCount=1, resolutionRule=StopOldest}
}
